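# Animation Blueprint for the SpitRomp female face accessory, layering procedural motion on the pose copied from the parent mesh.
# A Control Rig driven by crouch, crouch-walk and aim-up values feeds head-based RigidBody simulation, then spring-based AnimDynamics on dyn_hair_01.
# Each update derives aim-up from the head and spine_05 transform delta while targeting.
# It also stores parameter-asset gravity and rigid-body settings, though no AnimGraph pin visibly reads them.
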AnimBlueprint /BRCosmetics/Characters/Player/Female/Medium/Bodies/F_MED_SpitRomp/Meshes/Parts/F_MED_SpitRomp_FaceAcc_AnimBP
target_skeleton=&/F_MED_SpitRomp_FaceAcc_Skeleton
graphs [
 AnimGraph {nodes [
   n1 = Control Rig!(
    Source=Copy Pose From Mesh() {bUseAttachedParent=true}
    Alpha
    CrouchWalk:bool=Get crouch Walk() -> (crouch Walk:bool)
    Crouching:double=Get Crouch_Float() -> (Crouch_Float:double)
    Aim_Up:double=Get AimUp() -> (AimUp:double)
   ) -> (Pose) {controlRigAssetReference {blueprintRigClass=&/F_MED_SpitRomp_FaceAcc_CtrlRig.F_MED_SpitRomp_FaceAcc_CtrlRig_C}}
   n2 = RigidBody(ComponentPose=Local To Component(LocalPose=@n1)) {
    overridePhysicsAsset=&/F_MED_SpitRomp_FaceAcc_Physics
    alphaBoolBlend {blendOption=Linear}
    baseBoneRef {boneName=head}
   }
   n3 = Anim Dynamics - Bone: dyn_hair_01!(ComponentPose=@n2, Alpha=1, GravityOverride:Vector="0,0,-980", ExternalForce) -> (Pose) {
    angularDampingOverride=0.9999
    angularSpringConstant=2000
    bAngularSpring=true
    bOverrideAngularDamping=true
    bUseGravityOverride=true
    gravityScale=0.35
    simulationSpace=RootRelative
    alphaBoolBlend {blendOption=Linear}
    boundBone {boneName=dyn_hair_01}
    componentAppliedLinearAccClamp=<0 0 0>
    gravityOverride {z=-980}
    physicsBodyDefinitions [{
      boundBone {boneName=dyn_hair_01}
      boxExtents=<13 13 13>
      constraintSetup {angularLimitsMax=<10 25 25>, angularLimitsMin=<-10 -25 -25>, angularTarget=<1 0 0>}
      localJointOffset=<1 0 0>
     }]
   }
   Output Pose(Result=Component To Local(ComponentPose=@n3)) {layerGroup=None}
  ]}
 EventGraph {nodes [
   Event BlueprintUpdateAnimation()
   > n1 = Sequence!(execute) -> (then_0, then_1, then_2:Exec, then_3:Exec)
   > n2 = "Gravity Override State Function (Parameter Asset) (Nativized)"(
    CharacterPartAnimInstance=Self-Reference()
    GameTimeInSeconds=GetGameTimeInSeconds()
    Parameters=&/F_MED_SpitRomp_FaceAcc_Parameters
   )
   > Set GravityOverride(GravityOverride=@n2)
   > Set RBAN Settings Out(RBAN Settings Out=@n2.OutRigidBodyAnimNodeParameters)
   Set crouch Walk(execute=@n1.then_1, crouch Walk:bool=AND Boolean(A=Get bIsCrouching(), B=Get bIsMoving2D())) -> (Output_Get:bool)
   n3 = Set Head Up(
    execute=@n1.then_2
    Head Up:double=GetAlphaFromReferenceTransformDelta(
     SkeletalMeshComponent=Get Skeletal_Mesh() -> (Skeletal_Mesh:Skeletal Mesh Component Object Reference)
     BoneName=spine_05
     BaseName=head
     InputRange=(X=0,Y=90)
     TwistAxis="0,0,1"
    )
   ) -> (Output_Get:double)
   > Set AimUp(AimUp:double=SelectFloat(A=@n3, B=0, bPickA=Get bIsTargeting())) -> (Output_Get:double)
   Set Crouch_Float(execute=@n1.then_3, Crouch_Float:double="To Float (Boolean)"(InBool=Get bIsCrouching())) -> (Output_Get:double)
   Event BlueprintInitializeAnimation()
   > Set Skeletal_Mesh(Skeletal_Mesh:Skeletal Mesh Component Object Reference=GetOwnerSkeletalMeshComponent()) -> (Output_Get:Skeletal Mesh Component Object Reference)
  ]}
]
variables [
 AimUp {type=number}
 crouch Walk {type=boolean}
 Crouch_Float {type=number}
 GravityOverride {type=Vector}
 Head Up {type=number}
 Melee {type=number}
 pistol {type=number}
 pistol_0 {type=boolean}
 RBAN Settings Out {type=FortRigidBodyAnimNodeParameters}
 Riffle Equipped {type=boolean}
 Riffle Equipped alpha {type=number}
 RiffleAiming_Down {type=number}
 RiffleAimingUp {type=number}
 Rifle {type=number}
 Rifle_Bool {type=boolean}
 Shotgun {type=number}
 Shotgun_Bool {type=boolean}
 Skel Mesh {type=SkeletalMeshComponent}
 Skeletal_Mesh {type=SkeletalMeshComponent}
 Smg {type=number}
 Smg_Bool {type=boolean}
]
