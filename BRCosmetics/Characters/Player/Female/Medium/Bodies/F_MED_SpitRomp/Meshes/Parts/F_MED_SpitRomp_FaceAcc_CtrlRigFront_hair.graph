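# Control Rig for the F_MED_SpitRomp face accessory mesh that adjusts hair bones as the head rotates.
# A DeltaRotation function measures head rotation against head_null under spine_05, splits it into swing and twist, and remaps the swing's Z Euler angle to a clamped weight.
# AlphaInterp smooths that weight to blend in preset transform offsets on dynamic hair bones.
# The declared movement, crouch and aim variables are not read in the graph.
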ControlRigBlueprint /BRCosmetics/Characters/Player/Female/Medium/Bodies/F_MED_SpitRomp/Meshes/Parts/F_MED_SpitRomp_FaceAcc_CtrlRigFront_hair
graphs [
 New Function_ContainedGraph {nodes [
   n1 = RigVMFunctionEntryNode()
   RigVMFunction_MathBoolToInteger = &RigVMFunction_MathBoolToInteger(Value=@n1.onlyTwist)
   RigVMFunction_MathQuaternionSwingTwist = &RigVMFunction_MathQuaternionSwingTwist(
    TwistAxis=@n1.twistAxis
    Input=&RigUnit_GetRelativeTransformForItem(bChildInitial=False, bParentInitial=False, Child=@n1.child, Parent=@n1.parent).RelativeTransform.Rotation
   )
   n2 = RigVMDispatchNode(
    Values=((),())
    Index=@RigVMFunction_MathBoolToInteger.Result
    Values.0=@RigVMFunction_MathQuaternionSwingTwist.Swing
    Values.1=@RigVMFunction_MathQuaternionSwingTwist.Twist
   ) {template="DISPATCH_RigVMDispatch_SelectInt32(in Index,in Values,out Result)"}
   RigVMFunction_MathQuaternionToEuler = &RigVMFunction_MathQuaternionToEuler(Value=@n2.Result, RotationOrder=@n1.Rotation Order)
   > DISPATCH_RigVMDispatch_Print = RigVMDispatchNode(ScreenColor=(R=1,G=1,B=1,A=1), ScreenDuration=1.5, Enabled=@n1.debugRotation, Value=@RigVMFunction_MathQuaternionToEuler.Result) {template="DISPATCH_RigVMDispatch_Print(in Prefix,in Value,in Enabled,in ScreenDuration,in ScreenColor)"}
   Cast = RigVMDispatchNode(Value=@n1.triggerAxis) {template="DISPATCH_RigVMDispatch_CastEnumToInt(in Value,out Result)"}
   DISPATCH_RigVMDispatch_SelectInt32 = RigVMDispatchNode(
    Values=(0,0,0)
    Index=@Cast.Result
    Values.0=@RigVMFunction_MathQuaternionToEuler.Result.X
    Values.1=@RigVMFunction_MathQuaternionToEuler.Result.Y
    Values.2=@RigVMFunction_MathQuaternionToEuler.Result.Z
   ) {template="DISPATCH_RigVMDispatch_SelectInt32(in Index,in Values,out Result)"}
   Remap = &RigVMFunction_MathDoubleRemap(
    bClamp=False
    Value=@DISPATCH_RigVMDispatch_SelectInt32.Result
    TargetMinimum=@n1.outputMinMax.X
    TargetMaximum=@n1.outputMinMax.Y
    SourceMinimum=@n1.rotationMinMax.X
    SourceMaximum=@n1.rotationMinMax.Y
   )
   Absolute = &RigVMFunction_MathDoubleAbs(Value=@Remap.Result)
   Multiply = &RigVMFunction_MathDoubleMul(
    A=&RigVMFunction_MathDoubleClamp(Value=RigVMDispatchNode(True=@Absolute.Result, Condition=@n1.absoluteValue, False=@Remap.Result) {template="DISPATCH_RigVMDispatch_If(in Condition,in True,in False,out Result)"}.Result).Result
    B=@n1.premult
   )
   > RigVMFunctionReturnNode(outputQuat=@n2.Result, outputWeight=@Multiply.Result)
  ]}
 New Function_ContainedGraph {nodes [
   n1 = RigVMFunctionEntryNode()
   RigVMFunction_MathBoolToInteger = &RigVMFunction_MathBoolToInteger(Value=@n1.onlyTwist)
   RigVMFunction_MathQuaternionSwingTwist = &RigVMFunction_MathQuaternionSwingTwist(
    twistAxis=@n1.twistAxis
    Input=&RigUnit_GetRelativeTransformForItem(bChildInitial=False, bParentInitial=False, child=@n1.child, parent=@n1.parent).RelativeTransform.Rotation
   )
   n2 = RigVMDispatchNode(
    Values=((),())
    Index=@RigVMFunction_MathBoolToInteger.Result
    Values.0=@RigVMFunction_MathQuaternionSwingTwist.Swing
    Values.1=@RigVMFunction_MathQuaternionSwingTwist.Twist
   ) {template="DISPATCH_RigVMDispatch_SelectInt32(in Index,in Values,out Result)"}
   RigVMFunction_MathQuaternionToEuler = &RigVMFunction_MathQuaternionToEuler(Value=@n2.Result, RotationOrder=@n1.Rotation Order)
   > DISPATCH_RigVMDispatch_Print = RigVMDispatchNode(ScreenColor=(R=1,G=1,B=1,A=1), ScreenDuration=1.5, Enabled=@n1.debugRotation, Value=@RigVMFunction_MathQuaternionToEuler.Result) {template="DISPATCH_RigVMDispatch_Print(in Prefix,in Value,in Enabled,in ScreenDuration,in ScreenColor)"}
   Cast = RigVMDispatchNode(Value=@n1.triggerAxis) {template="DISPATCH_RigVMDispatch_CastEnumToInt(in Value,out Result)"}
   DISPATCH_RigVMDispatch_SelectInt32 = RigVMDispatchNode(
    Values=(0,0,0)
    Index=@Cast.Result
    Values.0=@RigVMFunction_MathQuaternionToEuler.Result.X
    Values.1=@RigVMFunction_MathQuaternionToEuler.Result.Y
    Values.2=@RigVMFunction_MathQuaternionToEuler.Result.Z
   ) {template="DISPATCH_RigVMDispatch_SelectInt32(in Index,in Values,out Result)"}
   Remap = &RigVMFunction_MathDoubleRemap(
    bClamp=False
    Value=@DISPATCH_RigVMDispatch_SelectInt32.Result
    TargetMinimum=@n1.outputMinMax.X
    TargetMaximum=@n1.outputMinMax.Y
    SourceMinimum=@n1.rotationMinMax.X
    SourceMaximum=@n1.rotationMinMax.Y
   )
   Absolute = &RigVMFunction_MathDoubleAbs(Value=@Remap.Result)
   Multiply = &RigVMFunction_MathDoubleMul(
    A=&RigVMFunction_MathDoubleClamp(Value=RigVMDispatchNode(True=@Absolute.Result, Condition=@n1.absoluteValue, False=@Remap.Result) {template="DISPATCH_RigVMDispatch_If(in Condition,in True,in False,out Result)"}.Result).Result
    B=@n1.premult
   )
   > RigVMFunctionReturnNode(outputQuat=@n2.Result, outputWeight=@Multiply.Result)
  ]}
 RigVMFunctionLibrary {nodes [
   DeltaRotation v4 = RigVMCollapseNode(
    ExecuteContext=()
    Rotation Order=ZYX
    child=(Type=Bone,Name="None")
    outputMinMax=(X=0,Y=1)
    parent=(Type=Bone,Name="None")
    premult=1
    rotationMinMax=(X=0,Y=0)
    twistAxis=(X=1,Y=0,Z=0)
   )
   DeltaRotation v4_1 = RigVMCollapseNode(
    ExecuteContext=()
    Rotation Order=ZYX
    child=(Type=Bone,Name="None")
    outputMinMax=(X=0,Y=1)
    parent=(Type=Bone,Name="None")
    premult=1
    rotationMinMax=(X=0,Y=0)
    twistAxis=(X=1,Y=0,Z=0)
   )
  ]}
 RigVMModel {nodes [
   RigUnit_BeginExecution = &RigUnit_BeginExecution()
   RigVMFunction_Sequence = &RigVMFunction_Sequence(ExecuteContext=@RigUnit_BeginExecution.ExecutePin)
   &RigUnit_ModifyTransforms(
    ItemToModify=((Item=(Type=Bone,Name="head"),Transform=(Rotation=(X=0,Y=0,Z=0,W=1),Translation=(X=0,Y=0,Z=0),Scale3D=(X=1,Y=1,Z=1))))
    Weight=1
    WeightMinimum=0
    ExecutePin=@RigVMFunction_Sequence.A
   )
   > DeltaRotation v4_2 = RigVMFunctionReferenceNode(
    Rotation Order=ZYX
    child=(Type=Bone,Name="head")
    outputMinMax=(X=0,Y=1)
    parent=(Type=Null,Name="head_null")
    premult=1
    rotationMinMax=(X=5,Y=10)
    triggerAxis=Z
    twistAxis=(X=1,Y=0,Z=0)
   )
   AlphaInterp = &RigVMFunction_AlphaInterp(InterpSpeedDecreasing=5, InterpSpeedIncreasing=5, Scale=1, bClampResult=False, Value=@"DeltaRotation v4_2".outputWeight)
   > &RigUnit_ModifyTransforms(
    ItemToModify="((Item=(Type=Bone,Name=\"None\"),Transform=(Rotation=(X=0,Y=0,Z=0,W=1),Translation=(X=0,Y=0,Z=0),Scale3D=(X=1,Y=1,Z=1))),(Item=(Type=Bone,Name=\"dyn_right_hair_b_01\"),Transform=(Rotation=(X=0,Y=0,Z=-0.1045,W=0.9945),Translation=(X=0,Y=0,Z=0),Scale3D=(X=1,Y=1, ...(truncated)"
    WeightMinimum=0
    Weight=@AlphaInterp.Result
   )
  ]}
]
hierarchy [
 root {loc=<0 0 0>, rot=<0 0 0 1>}
 pelvis {parent=root, loc=<0 -0.5301 98.0571>, rot=<0.0127 0.707 -0.0127 -0.707>}
 spine_01 {parent=pelvis, loc=<6.1208 -0.9328 0>, rot=<0 0 0.034 -0.9994>}
 spine_02 {parent=spine_01, loc=<6.5997 -0.2095 0>, rot=<0 0 0.0238 -0.9997>}
 spine_03 {parent=spine_02, loc=<6.4914 -0.1237 0>, rot=<0 0 0.0149 -0.9999>}
 spine_04 {parent=spine_03, loc=<6.9598 0.0794 0>, rot=<0 0 -0.0132 -0.9999>}
 spine_05 {parent=spine_04, loc=<6.9987 0 0>, rot=<0 0 -0.0015 -1>}
 neck_01 {parent=spine_05, loc=<11.3138 0.1529 0>, rot=<0 0 -0.2003 -0.9797>}
 neck_02 {parent=neck_01, loc=<4.5627 0 0>, rot=<0 0 0.0504 -0.9987>}
 head {parent=neck_02, loc=<5.3914 0 0>, rot=<0 0 0.115 -0.9934>}
 dyn_left_hair_base {parent=head, loc=<5.0333 4.1366 -7.4834>, rot=<0.0848 -0.9959 -0.0325 -0.0025>}
 dyn_left_hair_01 {parent=dyn_left_hair_base, loc=<0 0 0>, rot=<0 0 0 1>}
 dyn_left_hair_02 {parent=dyn_left_hair_01, loc=<4 0 0>, rot=<-0.0052 -0.1333 -0.0383 -0.9903>}
 dyn_left_hair_03 {parent=dyn_left_hair_02, loc=<4 0 0>, rot=<0.0025 0.065 -0.0386 -0.9971>}
 dyn_left_hair_04 {parent=dyn_left_hair_03, loc=<4 0 0>, rot=<-0.0104 0.1942 -0.0417 -0.98>}
 FACIAL_C_Jaw {parent=head, loc=<-0.4271 1.5514 0>, rot=<-0.6315 0.3182 0.6315 0.3182>}
 FACIAL_C_Jawline {parent=FACIAL_C_Jaw, loc=<-0.0004 2.6936 6.8544>, rot=<0.2612 0 0 -0.9653>}
 FACIAL_L_ChinSide {parent=FACIAL_C_Jaw, loc=<2.6155 0.8031 6.7974>, rot=<-0.2198 -0.3737 0.0503 -0.8997>}
 FACIAL_R_ChinSide {parent=FACIAL_C_Jaw, loc=<-2.6151 0.8027 6.7973>, rot=<-0.2198 0.3738 -0.0503 -0.8997>}
 FACIAL_C_Chin {parent=FACIAL_C_Jaw, loc=<-0.0005 1.1571 7.3475>, rot=<-0.2568 0 0 -0.9665>}
 dyn_hair_base {parent=head, loc=<10.1973 -8.8477 0>, rot=<0.5654 -0.4246 0.4246 -0.5654>}
 dyn_hair_01 {parent=dyn_hair_base, loc=<0 0 0>, rot=<0 0 0 1>}
 dyn_right_hair_b_base {parent=head, loc=<3.7327 8.2654 4.1739>, rot=<0.0545 -0.0036 0.9957 -0.0747>}
 dyn_right_hair_b_01 {parent=dyn_right_hair_b_base, loc=<0 0 0>, rot=<0 0 0 1>}
 dyn_right_hair_b_02 {parent=dyn_right_hair_b_01, loc=<5 0 0>, rot=<-0.0034 -0.1404 -0.0236 -0.9898>}
 dyn_right_hair_b_03 {parent=dyn_right_hair_b_02, loc=<5 0 0>, rot=<-0.0062 -0.0667 0.0928 -0.9934>}
 dyn_right_hair_a_base_Parent {parent=head, loc=<8.2582 6.2218 5.2608>, rot=<0.1336 0.1435 0.9622 -0.189>}
 dyn_right_hair_a_base {parent=dyn_right_hair_a_base_Parent, loc=<0 -0.0198 0>, rot=<0 0 0 1>}
 dyn_right_hair_a_01 {parent=dyn_right_hair_a_base, loc=<3.1956 0 0.1945>, rot=<0.0029 -0.1134 0.0359 -0.9929>}
 dyn_right_hair_a_02 {parent=dyn_right_hair_a_01, loc=<4.0003 0 0>, rot=<0.0022 -0.0466 0.0956 -0.9943>}
 dyn_right_hair_a_03 {parent=dyn_right_hair_a_02, loc=<4 0 0>, rot=<0.0038 -0.0408 0.0923 -0.9949>}
 dyn_right_hair_a_04 {parent=dyn_right_hair_a_03, loc=<4 0 0>, rot=<0 0 0.1649 -0.9863>}
 dyn_chain {parent=neck_01, loc=<0.2208 5.699 0.0003>, rot=<-0.194 -0.68 -0.68 -0.194>}
 clavicle_l {parent=spine_05, loc=<5.4831 -0.1053 -3.2946>, rot=<0.0465 -0.7289 -0.009 -0.6829>}
 upperarm_l {parent=clavicle_l, loc=<9.8658 -1.2136 -0.4105>, rot=<0.0164 -0.3245 -0.0269 -0.9454>}
 upperarm_twist_01_l {parent=upperarm_l, loc=<7.2135 -0.1134 0>, rot=<0 0 0 1>}
 deform_upperarm_up_base_l {parent=upperarm_twist_01_l, loc=<-8.891 0.4825 3.9517>, rot=<0 0 0 -1>}
 deform_upperarm_up_l {parent=deform_upperarm_up_base_l, loc=<0 0 0>, rot=<0 0 0 1>}
 clavicle_r {parent=spine_05, loc=<5.4831 -0.1053 3.2946>, rot=<0.7289 0.0465 -0.6829 0.009>}
 upperarm_r {parent=clavicle_r, loc=<-9.8658 1.2136 0.4105>, rot=<0.0164 -0.3245 -0.0269 -0.9454>}
 upperarm_twist_01_r {parent=upperarm_r, loc=<-7.2135 0.1134 0>, rot=<0 0 0 -1>}
 deform_upperarm_up_base_r {parent=upperarm_twist_01_r, loc=<8.891 -0.4825 -3.9517>, rot=<0 0 0 -1>}
 deform_upperarm_up_r {parent=deform_upperarm_up_base_r, loc=<0 0 0>, rot=<0 0 0 1>}
 head_null {parent=spine_05, type=Null, loc=<20.6573 3.5489 0>, rot=<0 0 -0.036 -0.9994>}
 dyn_headphone_Back {parent=spine_05, loc=<14.8823 -3.302 -0.0064>, rot=<-0.2464 -0.6981 -0.4111 -0.5319>}
 dyn_headphone_right {parent=dyn_headphone_Back, loc=<-1.2286 2.6869 -0.3118>, rot=<0.0385 -0.043 -0.7436 -0.6662>}
 dyn_headphone_left {parent=dyn_headphone_Back, loc=<1.5424 -1.7998 0.4272>, rot=<0.1055 0.0012 0.2772 -0.955>}
]
variables [
 Aim_Up {type=double, default=0}
 CrouchWalk {type=bool, default=False}
 Crouching {type=double, default=0}
 Move_Forword {type=double, default=0}
 Move_backword {type=double, default=0}
 Pelvis_Prox {type=float, default=0}
]
