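# Physics asset for the SpitRomp female medium character's headphone face-accessory mesh, defining capsule bodies and joints for its dynamic hair bones.
# One left and two right hair chains hang from base bodies on the head, with tight swing limits and locked or zero-degree twist.
# Head, upper spine, clavicle and upper-arm capsules form a body proxy with free joints.
# Both right bases are locked to the head, while the left base is free.
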PhysicsAsset /BRCosmetics/Characters/Player/Female/Medium/Bodies/F_MED_SpitRomp/Meshes/Parts/F_MED_SpitRomp_FaceAcc_Headphone_Physics
bodies {
 clavicle_l [{shapeType=Capsule, length=10, radius=7, shapeName=clavicle_l_capsule, center=<5 0 -1.5>, rotation=<rot 90 0 0>}]
 clavicle_r [{shapeType=Capsule, length=10, radius=7, shapeName=clavicle_r_capsule, center=<-5.0628 0 1.54>, rotation=<rot 88.0216 -179.9975 -179.9975>}]
 dyn_left_hair_01 [{shapeType=Capsule, length=2, radius=1, shapeName=dyn_left_hair_01_capsule, center=<1 0 0.3>, rotation=<rot 90 0.2714 0>}]
 dyn_left_hair_02 [{shapeType=Capsule, length=2, radius=1, shapeName=dyn_left_hair_02_capsule, center=<1 0 0.3>, rotation=<rot 83.2098 180 179.9999>}]
 dyn_left_hair_03 [{shapeType=Capsule, length=2, radius=1, shapeName=dyn_left_hair_03_capsule, center=<1 0 0.5>, rotation=<rot 90 0 0>}]
 dyn_left_hair_04 [{shapeType=Capsule, length=2, radius=1, shapeName=dyn_left_hair_04_capsule, center=<1 0.7 -0.3>, rotation=<rot 90 0 0>}]
 dyn_left_hair_base [{shapeType=Capsule, length=0.505, radius=0.505, shapeName=dyn_left_hair_base_capsule, rotation=<rot 90 0 0>}]
 dyn_right_hair_a_01 [{shapeType=Capsule, length=2.5, radius=1.3, shapeName=dyn_right_hair_a_01_capsule, center=<2 0 1.2>, rotation=<rot 90 0 0>}]
 dyn_right_hair_a_02 [{shapeType=Capsule, length=2.5, radius=1.1, shapeName=dyn_right_hair_a_02_capsule, center=<3 0 1.2>, rotation=<rot 90 -0.0001 -0.0001>}]
 dyn_right_hair_a_03 [{shapeType=Capsule, length=2.5, radius=0.8, shapeName=dyn_right_hair_a_03_capsule, center=<2 0 0.4>, rotation=<rot 90 0 0>}]
 dyn_right_hair_a_04 [{shapeType=Capsule, length=2, radius=0.5, shapeName=dyn_right_hair_a_04_capsule, center=<2 0 -0.5>, rotation=<rot 90 0 0>}]
 dyn_right_hair_a_base [{shapeType=Capsule, length=0.505, radius=0.505, shapeName=dyn_right_hair_a_base_capsule, rotation=<rot 90 0 0>}]
 dyn_right_hair_b_01 [{shapeType=Capsule, length=3, radius=1.5, shapeName=dyn_right_hair_b_01_capsule, center=<3 0 2>, rotation=<rot 77.5535 -0.0001 0.0001>}]
 dyn_right_hair_b_02 [{shapeType=Capsule, length=3, radius=1.5, shapeName=dyn_right_hair_b_02_capsule, center=<3 0 1.7>, rotation=<rot 90 0 0>}]
 dyn_right_hair_b_base [{shapeType=Capsule, length=0.505, radius=0.505, shapeName=dyn_right_hair_b_base_capsule, rotation=<rot 90 0 0>}]
 head [{shapeType=Capsule, length=10, radius=6, shapeName=head_capsule, center=<2.1332 4.6836 -0.0761>, rotation=<rot 73.1894 -55.0517 -83.8121>}]
 spine_05 [{shapeType=Capsule, length=10, radius=7, shapeName=spine_05_capsule, center=<4.5 -4 0>}]
 upperarm_l [{shapeType=Capsule, length=14, radius=5.5, shapeName=upperarm_l_capsule, center=<14 0 0>, rotation=<rot 90 0 0>}]
 upperarm_r [{shapeType=Capsule, length=14, radius=5.5, shapeName=upperarm_r_capsule, center=<-14 0 0>, rotation=<rot 90 0 0>}]
}
constraints [
 {
  bone1Name=dyn_left_hair_03
  bone2Name=dyn_left_hair_02
  swing1LimitDegrees=13
  swing2LimitDegrees=13
  twistLimitDegrees=45
  twistMotion=Locked
 }
 {
  bone1Name=dyn_left_hair_02
  bone2Name=dyn_left_hair_01
  swing1LimitDegrees=13
  swing2LimitDegrees=13
  twistLimitDegrees=45
  twistMotion=Locked
 }
 {
  bone1Name=dyn_left_hair_01
  bone2Name=dyn_left_hair_base
  swing1LimitDegrees=5
  swing2LimitDegrees=5
  twistLimitDegrees=45
  twistMotion=Locked
 }
 {
  bone1Name=dyn_left_hair_base
  bone2Name=head
  swing1LimitDegrees=45
  swing1Motion=Free
  swing2LimitDegrees=45
  swing2Motion=Free
  twistLimitDegrees=45
  twistMotion=Free
 }
 {
  bone1Name=dyn_right_hair_b_base
  bone2Name=head
  swing1LimitDegrees=45
  swing1Motion=Locked
  swing2LimitDegrees=45
  swing2Motion=Locked
  twistLimitDegrees=45
  twistMotion=Locked
 }
 {bone1Name=dyn_right_hair_b_01, bone2Name=dyn_right_hair_b_base, swing1LimitDegrees=10, swing2LimitDegrees=10, twistLimitDegrees=0}
 {bone1Name=dyn_right_hair_b_02, bone2Name=dyn_right_hair_b_01, swing1LimitDegrees=10, swing2LimitDegrees=10, twistLimitDegrees=0}
 {
  bone1Name=upperarm_r
  bone2Name=clavicle_r
  swing1LimitDegrees=45
  swing1Motion=Free
  swing2LimitDegrees=45
  swing2Motion=Free
  twistLimitDegrees=45
  twistMotion=Free
 }
 {
  bone1Name=upperarm_l
  bone2Name=clavicle_l
  swing1LimitDegrees=45
  swing1Motion=Free
  swing2LimitDegrees=45
  swing2Motion=Free
  twistLimitDegrees=45
  twistMotion=Free
 }
 {
  bone1Name=clavicle_l
  bone2Name=spine_05
  swing1LimitDegrees=45
  swing1Motion=Free
  swing2LimitDegrees=45
  swing2Motion=Free
  twistLimitDegrees=45
  twistMotion=Free
 }
 {
  bone1Name=clavicle_r
  bone2Name=spine_05
  swing1LimitDegrees=45
  swing1Motion=Free
  swing2LimitDegrees=45
  swing2Motion=Free
  twistLimitDegrees=45
  twistMotion=Free
 }
 {
  bone1Name=head
  bone2Name=spine_05
  swing1LimitDegrees=45
  swing1Motion=Free
  swing2LimitDegrees=45
  swing2Motion=Free
  twistLimitDegrees=45
  twistMotion=Free
 }
 {bone1Name=dyn_left_hair_04, bone2Name=dyn_left_hair_03, swing1LimitDegrees=13, swing2LimitDegrees=13, twistLimitDegrees=0}
 {
  bone1Name=dyn_right_hair_a_base
  bone2Name=head
  swing1LimitDegrees=45
  swing1Motion=Locked
  swing2LimitDegrees=45
  swing2Motion=Locked
  twistLimitDegrees=45
  twistMotion=Locked
 }
 {bone1Name=dyn_right_hair_a_01, bone2Name=dyn_right_hair_a_base, swing1LimitDegrees=10, swing2LimitDegrees=10, twistLimitDegrees=0}
 {bone1Name=dyn_right_hair_a_02, bone2Name=dyn_right_hair_a_01, swing1LimitDegrees=13, swing2LimitDegrees=13, twistLimitDegrees=0}
 {bone1Name=dyn_right_hair_a_03, bone2Name=dyn_right_hair_a_02, swing1LimitDegrees=13, swing2LimitDegrees=13, twistLimitDegrees=0}
 {bone1Name=dyn_right_hair_a_04, bone2Name=dyn_right_hair_a_03, swing1LimitDegrees=13, swing2LimitDegrees=13, twistLimitDegrees=0}
]
