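# Post-process-style Anim Blueprint for the SpitRomp face accessory, animating hair and headphones on the player cosmetic.
# Copies the parent mesh pose into a Control Rig driven by crouch and aim values, then applies Anim Dynamics to dyn_hair_01.
# A Headphone bool, set from the Mat2 material variant, picks between two neck_01-space RigidBody physics assets before a front-hair Control Rig.
# Updates also compute gravity and RBAN settings, which the graph leaves unconnected.
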AnimBlueprint /BRCosmetics/Characters/Player/Female/Medium/Bodies/F_MED_SpitRomp/Meshes/Parts/F_MED_SpitRomp_FaceAcc_Skeleton_AnimBlueprin
target_skeleton=&/F_MED_SpitRomp_FaceAcc_Skeleton
graphs [
 AnimGraph {nodes [
   n1 = RigidBody!(
    ComponentPose=Local To Component(LocalPose=Use cached pose 'StartPose'())
    Alpha
    OverrideWorldGravity:Vector="0,0,0"
    ExternalForce
    SimSpaceSettings:Sim Space Settings Structure=(WorldAlpha=0,VelocityScaleZ=1,DampingAlpha=1,MaxLinearVelocity=10000,MaxAngularVelocity=10000,MaxLinearAcceleration=10000,MaxAngularAcceleration=10000,ExternalLinearDragV=(X=0,Y=0,Z=0),ExternalLinearVelocity=(X=0,Y=0,Z=0),ExternalAngularVelocity=(X=0,Y=0,Z=0))
   ) -> (Pose) {
    bOverrideWorldGravity=true
    overridePhysicsAsset=&/F_MED_SpitRomp_FaceAcc_Physics
    simulationSpace=BaseBoneSpace
    alphaScaleBiasClamp {bInterpResult=true, interpSpeedDecreasing=5}
    baseBoneRef {boneName=neck_01}
   }
   n2 = RigidBody(
    ComponentPose=Local To Component(LocalPose=Use cached pose 'StartPose'())
    OverrideWorldGravity="0,0,0"
    SimSpaceSettings=(WorldAlpha=0,VelocityScaleZ=1,DampingAlpha=1,MaxLinearVelocity=10000,MaxAngularVelocity=10000,MaxLinearAcceleration=10000,MaxAngularAcceleration=10000,ExternalLinearDragV=(X=0,Y=0,Z=0),ExternalLinearVelocity=(X=0,Y=0,Z=0),ExternalAngularVelocity=(X=0,Y=0,Z=0))
   ) {
    bDefaultToSkeletalMeshPhysicsAsset=false
    bOverrideWorldGravity=true
    overridePhysicsAsset=&/F_MED_SpitRomp_FaceAcc_Headphone_Physics
    simulationSpace=BaseBoneSpace
    alphaScaleBiasClamp {interpSpeedDecreasing=5}
    baseBoneRef {boneName=neck_01}
   }
   n3 = Control Rig(Source=Blend Poses by bool(
     BlendPose_0=Component To Local(ComponentPose=@n2)
     BlendPose_1=Component To Local(ComponentPose=@n1)
     bActiveValue=Get Headphone() -> (Headphone:bool)
    )) {controlRigAssetReference {blueprintRigClass=&/F_MED_SpitRomp_FaceAcc_CtrlRigFront_hair.F_MED_SpitRomp_FaceAcc_CtrlRigFront_hair_C}}
   Output Pose(Result=@n3)
   n4 = Control Rig!(
    Source=Copy Pose From Mesh() {bUseAttachedParent=true}
    Alpha
    CrouchWalk:bool=Get crouch Walk() -> (crouch Walk:bool)
    Crouching:double=Get Crouch_Float() -> (Crouch_Float:double)
    Aim_Up:double=Get AimUp() -> (AimUp:double)
    Move_Forword:double=Get moveForward() -> (moveForward:double)
    Move_backword:double=Get moveBackward() -> (moveBackward:double)
   ) -> (Pose) {controlRigAssetReference {blueprintRigClass=&/F_MED_SpitRomp_FaceAcc_CtrlRig.F_MED_SpitRomp_FaceAcc_CtrlRig_C}}
   n5 = Anim Dynamics - Bone: dyn_hair_01!(ComponentPose=Local To Component(LocalPose=@n4), Alpha=1, GravityOverride:Vector="0,0,-980", ExternalForce) -> (Pose) {
    angularDampingOverride=1
    angularSpringConstant=20000
    bAngularSpring=true
    bOverrideAngularDamping=true
    bUseGravityOverride=true
    gravityScale=0.35
    simulationSpace=RootRelative
    boundBone {boneName=dyn_hair_01}
    componentAppliedLinearAccClamp=<5000 5000 5000>
    componentLinearAccScale=<0.5 0.5 0.5>
    componentLinearVelScale=<0.5 0.5 0.5>
    gravityOverride {z=-980}
    physicsBodyDefinitions [{
      boundBone {boneName=dyn_hair_01}
      boxExtents=<20 30 30>
      constraintSetup {angularLimitsMax=<0 40 40>, angularLimitsMin=<0 -40 -40>, angularTarget=<1 0 0>}
      localJointOffset=<4 0 0>
     }]
   }
   Save cached pose 'StartPose'(Pose=Component To Local(ComponentPose=@n5)) {cacheName=StartPose}
  ]}
 EventGraph {nodes [
   Event BlueprintUpdateAnimation()
   > n1 = Sequence!(execute) -> (then_0, then_1, then_2:Exec, then_3:Exec, then_4:Exec)
   > n2 = "Gravity Override State Function (Parameter Asset) (Nativized)"(
    CharacterPartAnimInstance=Self-Reference()
    GameTimeInSeconds=GetGameTimeInSeconds()
    Parameters=&/F_MED_SpitRomp_FaceAcc_Parameters
   )
   > Set GravityOverride(GravityOverride=@n2)
   > Set RBAN Settings Out(RBAN Settings Out=@n2.OutRigidBodyAnimNodeParameters)
   Set crouch Walk(execute=@n1.then_1, crouch Walk:bool=AND Boolean(A=Get bIsCrouching(), B=Get bIsMoving2D())) -> (Output_Get:bool)
   n3 = Set Head Up(
    execute=@n1.then_2
    Head Up:double=GetAlphaFromReferenceTransformDelta(
     SkeletalMeshComponent=Get Skeletal_Mesh() -> (Skeletal_Mesh:Skeletal Mesh Component Object Reference)
     BoneName=spine_05
     BaseName=head
     InputRange=(X=0,Y=90)
     TwistAxis="0,0,1"
    )
   ) -> (Output_Get:double)
   > Set AimUp(AimUp:double=SelectFloat(A=@n3, B=0, bPickA=Get bIsTargeting())) -> (Output_Get:double)
   Set Crouch_Float(execute=@n1.then_3, Crouch_Float:double="To Float (Boolean)"(InBool=Get bIsCrouching())) -> (Output_Get:double)
   n4 = Cast To FortPlayerPawn(execute=@n1.then_4, Object=GetOwningActor())
   > n5 = GetVariantChannelActiveIndexForItemActor!(
    execute
    Item:Athena Cosmetic Item Definition Object Reference=/BRCosmetics/Athena/Items/Cosmetics/Characters/Character_SpitRomp
    Actor:Actor Object Reference=@n4
    Channel:Gameplay Tag Structure=(TagName="Cosmetics.Variant.Channel.Material")
    VariantList:Array of Gameplay Tag Structures=Make Array("[0]"=(TagName="Cosmetics.Variant.Property.Mat2"))
    bErrorIfChannelMissing:bool=false
    bErrorIfTagNotInArray:bool=false
   ) -> (then, Index:int=0, bIndexIsValid:bool, bFoundChannel:bool)
   > n6 = Branch(Condition=@n5.bIndexIsValid)
   > Set Headphone(Headphone:bool=true) -> (Output_Get:bool)
   Set Headphone(execute=@n6.else, Headphone=false)
   Event BlueprintInitializeAnimation()
   > Set Skeletal_Mesh(Skeletal_Mesh:Skeletal Mesh Component Object Reference=GetOwnerSkeletalMeshComponent()) -> (Output_Get:Skeletal Mesh Component Object Reference)
   > Set Headphone(Headphone=true)
   TryGetPawnOwner()
   Get Skeletal_Mesh()
  ]}
]
variables [
 AimUp {type=number}
 crouch Walk {type=boolean}
 Crouch_Float {type=number}
 GravityOverride {type=Vector}
 Head Up {type=number}
 Headphone {type=boolean}
 "Move Left " {type=number}
 Move Right {type=number}
 moveBackward {type=number}
 moveForward {type=number}
 RBAN Settings Out {type=FortRigidBodyAnimNodeParameters}
 Skeletal_Mesh {type=SkeletalMeshComponent}
]
